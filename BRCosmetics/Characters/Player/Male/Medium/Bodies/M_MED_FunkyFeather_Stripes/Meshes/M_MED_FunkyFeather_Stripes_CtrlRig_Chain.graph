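# Control Rig for the FunkyFeather Stripes male body cosmetic that correctively offsets its leg chains, T-shirt and backpack bones.
# An aim bone tracks the right calf as reference for a thigh pose reader.
# That reader, a thigh-to-shirt distance remap and the chain's local rotation weight additive chain offsets.
# When IsFrontend is true, a pelvis pose reader bends shirt bones, and the backpack gets a fixed offset plus spine-driven tilt.
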ControlRigBlueprint /BRCosmetics/Characters/Player/Male/Medium/Bodies/M_MED_FunkyFeather_Stripes/Meshes/M_MED_FunkyFeather_Stripes_CtrlRig_Chain
graphs [
 RigVMFunctionLibrary {links [], nodes []}
 RigVMModel {nodes [
   RigUnit_BeginExecution = &RigUnit_BeginExecution()
   RigVMFunction_Sequence = &RigVMFunction_Sequence(ExecuteContext=@RigUnit_BeginExecution.ExecutePin)
   &RigUnit_AimConstraintLocalSpaceOffset(
    Child=(Type=Bone,Name="thigh_r_aim")
    Filter=(bX=True,bY=True,bZ=True)
    Parents=((Item=(Type=Bone,Name="calf_r"),Weight=1))
    UpAxis=(X=0,Y=1,Z=0)
    Weight=1
    WorldUp=(Target=(X=0,Y=0,Z=1),Kind=Direction,Space=(Type=Bone,Name="pelvis"))
    ExecutePin=@RigVMFunction_Sequence.A
   )
   > n1 = &RigUnit_SphericalPoseReader(
    ActiveRegionScaleFactors=(PositiveWidth=1,NegativeWidth=1,PositiveHeight=1,NegativeHeight=1)
    ActiveRegionSize=0.38
    Debug=(bDrawDebug=false,bDraw2D=False,bDrawLocalAxes=False,DebugScale=25,DebugSegments=20,DebugThickness=0.25)
    DriverAxis=(X=0,Y=1,Z=0)
    DriverItem=(Type=Bone,Name="thigh_twist_01_r")
    FalloffRegionScaleFactors=(PositiveWidth=1,NegativeWidth=1,PositiveHeight=1,NegativeHeight=1)
    FalloffSize=0.5
    FlipHeightScaling=False
    FlipWidthScaling=False
    OptionalParentItem=(Type=Bone,Name="thigh_r_aim")
    RotationOffset=(X=0,Y=0,Z=0)
   )
   > &RigUnit_OffsetTransformForItem(
    Item=(Type=Bone,Name="chain_leg_A_2")
    OffsetTransform=(Rotation=(X=0,Y=0,Z=0,W=1),Translation=(X=-1,Y=0,Z=0.8),Scale3D=(X=1,Y=1,Z=1))
    bPropagateToChildren=True
    Weight=@n1.OutputParam
   )
   > &RigUnit_OffsetTransformForItem(
    Item=(Type=Bone,Name="chain_leg_B_2")
    OffsetTransform=(Rotation=(X=0,Y=0,Z=0,W=1),Translation=(X=-2,Y=0,Z=0),Scale3D=(X=1,Y=1,Z=1))
    bPropagateToChildren=True
    Weight=@n1.OutputParam
   )
   RigUnit_GetTransform = &RigUnit_GetTransform(Item=(Type=Bone,Name="twist_01_r_pnt"), bInitial=False)
   RigVMFunction_MathVectorDistance = &RigVMFunction_MathVectorDistance(
    A=@RigUnit_GetTransform.Transform.Translation
    B=&RigUnit_GetTransform(Item=(Type=Bone,Name="tshirt_A_l_dyn_pnt"), bInitial=False).Transform.Translation
   )
   Remap = &RigVMFunction_MathDoubleRemap(SourceMaximum=2, SourceMinimum=6, TargetMaximum=1, TargetMinimum=0, Value=@RigVMFunction_MathVectorDistance.Result)
   > &RigUnit_OffsetTransformForItem(
    Item=(Type=Bone,Name="chain_leg_A_2")
    OffsetTransform=(Rotation=(X=0,Y=0,Z=0,W=1),Translation=(X=0,Y=0,Z=1),Scale3D=(X=1,Y=1,Z=1))
    bPropagateToChildren=True
    Weight=@Remap.Result
   )
   > &RigUnit_OffsetTransformForItem(
    Item=(Type=Bone,Name="chain_leg_B_2")
    OffsetTransform=(Rotation=(X=0,Y=0,Z=0,W=1),Translation=(X=-0.8,Y=0,Z=0),Scale3D=(X=1,Y=1,Z=1))
    bPropagateToChildren=True
    Weight=@Remap.Result
   )
   RigVMFunction_MathQuaternionToEuler = &RigVMFunction_MathQuaternionToEuler(
    RotationOrder=ZYX
    Value=&RigUnit_GetTransform(Item=(Type=Bone,Name="chain_leg_A_1"), Space=LocalSpace, bInitial=False).Transform.Rotation
   )
   n2 = &RigVMFunction_MathDoubleRemap(SourceMaximum=-30, SourceMinimum=0, TargetMaximum=1, TargetMinimum=0, Value=@RigVMFunction_MathQuaternionToEuler.Result.Z)
   > &RigUnit_OffsetTransformForItem(
    Item=(Type=Bone,Name="chain_leg_A_2")
    OffsetTransform=(Rotation=(X=0,Y=0,Z=0,W=1),Translation=(X=0.8,Y=0,Z=0),Scale3D=(X=1,Y=1,Z=1))
    bPropagateToChildren=True
    Weight=@n2.Result
   )
   > &RigUnit_OffsetTransformForItem(
    Item=(Type=Bone,Name="chain_leg_B_2")
    OffsetTransform=(Rotation=(X=0,Y=0,Z=0,W=1),Translation=(X=0.5,Y=0,Z=0),Scale3D=(X=1,Y=1,Z=1))
    bPropagateToChildren=True
    Weight=@n2.Result
   )
   RigVMFunction_ControlFlowBranch = &RigVMFunction_ControlFlowBranch(ExecuteContext=@RigVMFunction_Sequence.B, Condition=RigVMVariableNode() {variable=IsFrontend}.Value)
   n3 = &RigUnit_SphericalPoseReader(
    ActiveRegionScaleFactors=(PositiveWidth=1,NegativeWidth=1,PositiveHeight=1,NegativeHeight=1)
    ActiveRegionSize=0.2
    Debug=(bDrawDebug=True,bDraw2D=False,bDrawLocalAxes=False,DebugScale=25,DebugSegments=20,DebugThickness=0.25)
    DriverAxis=(X=0,Y=-1,Z=0)
    DriverItem=(Type=Bone,Name="pelvis")
    FalloffRegionScaleFactors=(PositiveWidth=1,NegativeWidth=1,PositiveHeight=1,NegativeHeight=1)
    FalloffSize=0.4
    FlipHeightScaling=False
    FlipWidthScaling=False
    OptionalParentItem=(Type=Bone,Name="root")
    RotationOffset=(X=0,Y=90,Z=0)
    ExecutePin=@RigVMFunction_ControlFlowBranch.True
   )
   > &RigUnit_OffsetTransformForItem(
    Item=(Type=Bone,Name="tshirt_A_l_dyn_4")
    OffsetTransform=(Rotation=(X=0,Y=0.1219,Z=0,W=0.9925),Translation=(X=0,Y=0,Z=0),Scale3D=(X=1,Y=1,Z=1))
    bPropagateToChildren=True
    Weight=@n3.OutputParam
   )
   > &RigUnit_OffsetTransformForItem(
    Item=(Type=Bone,Name="tshirt_A_l_dyn_21")
    OffsetTransform=(Rotation=(X=0,Y=0.0698,Z=0,W=0.9976),Translation=(X=0,Y=0,Z=0),Scale3D=(X=1,Y=1,Z=1))
    bPropagateToChildren=True
    Weight=@n3.OutputParam
   )
   > &RigUnit_OffsetTransformForItem(
    Item=(Type=Bone,Name="tshirt_A_l_dyn_31")
    OffsetTransform=(Rotation=(X=0,Y=-0.2419,Z=0,W=0.9703),Translation=(X=0,Y=0,Z=0),Scale3D=(X=1,Y=1,Z=1))
    bPropagateToChildren=True
    Weight=@n3.OutputParam
   )
   > &RigUnit_OffsetTransformForItem(
    Item=(Type=Bone,Name="tshirt_mid_l_dyn_1")
    OffsetTransform=(Rotation=(X=0,Y=0.1045,Z=0,W=0.9945),Translation=(X=0,Y=0,Z=0),Scale3D=(X=1,Y=1,Z=1))
    bPropagateToChildren=True
    Weight=@n3.OutputParam
   )
   > &RigUnit_OffsetTransformForItem(
    Item=(Type=Bone,Name="tshirt_mid_l_dyn_2")
    OffsetTransform=(Rotation=(X=0,Y=0.0698,Z=0,W=0.9976),Translation=(X=0,Y=0,Z=0),Scale3D=(X=1,Y=1,Z=1))
    bPropagateToChildren=True
    Weight=@n3.OutputParam
   )
   > &RigUnit_OffsetTransformForItem(
    Item=(Type=Bone,Name="tshirt_mid_l_dyn_3")
    OffsetTransform=(Rotation=(X=0,Y=-0.2419,Z=0,W=0.9703),Translation=(X=0,Y=0,Z=0),Scale3D=(X=1,Y=1,Z=1))
    bPropagateToChildren=True
    Weight=@n3.OutputParam
   )
   > &RigUnit_OffsetTransformForItem(
    Item=(Type=Bone,Name="tshirt_A_r_dyn_1")
    OffsetTransform=(Rotation=(X=0,Y=0.1219,Z=0,W=0.9925),Translation=(X=0,Y=0,Z=0),Scale3D=(X=1,Y=1,Z=1))
    bPropagateToChildren=True
    Weight=@n3.OutputParam
   )
   > &RigUnit_OffsetTransformForItem(
    Item=(Type=Bone,Name="tshirt_A_r_dyn_2")
    OffsetTransform=(Rotation=(X=0,Y=0.0698,Z=0,W=0.9976),Translation=(X=0,Y=0,Z=0),Scale3D=(X=1,Y=1,Z=1))
    bPropagateToChildren=True
    Weight=@n3.OutputParam
   )
   > &RigUnit_OffsetTransformForItem(
    Item=(Type=Bone,Name="tshirt_A_r_dyn_3")
    OffsetTransform=(Rotation=(X=0,Y=-0.2419,Z=0,W=0.9703),Translation=(X=0,Y=0,Z=0),Scale3D=(X=1,Y=1,Z=1))
    bPropagateToChildren=True
    Weight=@n3.OutputParam
   )
   > &RigUnit_OffsetTransformForItem(
    Item=(Type=Bone,Name="attach_backpack")
    OffsetTransform=(Rotation=(X=0.0262,Y=0,Z=0,W=0.9997),Translation=(X=0,Y=0.6,Z=0),Scale3D=(X=1,Y=1,Z=1))
    Weight=1
    bPropagateToChildren=True
   )
   > n4 = &RigUnit_SphericalPoseReader(
    ActiveRegionScaleFactors=(PositiveWidth=1,NegativeWidth=1,PositiveHeight=1,NegativeHeight=1)
    ActiveRegionSize=0.43
    Debug=(bDrawDebug=True,bDraw2D=False,bDrawLocalAxes=False,DebugScale=25,DebugSegments=20,DebugThickness=0.25)
    DriverAxis=(X=1,Y=0,Z=0)
    DriverItem=(Type=Bone,Name="spine_05")
    FalloffRegionScaleFactors=(PositiveWidth=1,NegativeWidth=1,PositiveHeight=1,NegativeHeight=1)
    FalloffSize=0.5
    FlipHeightScaling=False
    FlipWidthScaling=False
    OptionalParentItem=(Type=Bone,Name="None")
    RotationOffset=(X=-90,Y=0,Z=7)
   )
   > &RigUnit_OffsetTransformForItem(
    Item=(Type=Bone,Name="attach_backpack")
    OffsetTransform=(Rotation=(X=-0.0872,Y=0,Z=0,W=0.9962),Translation=(X=0,Y=0,Z=0),Scale3D=(X=1,Y=1,Z=1))
    bPropagateToChildren=True
    Weight=@n4.OutputParam
   )
  ]}
]
hierarchy [
 root {loc=<0 0 0>, rot=<0 0 0 1>}
 pelvis {parent=root, loc=<0 -0.7071 93.1191>, rot=<0.0127 0.707 -0.0127 -0.707>}
 spine_01 {parent=pelvis, loc=<6.7494 -0.6388 0>, rot=<0 0 0.034 -0.9994>}
 spine_02 {parent=spine_01, loc=<7.2518 -0.035 0>, rot=<0 0 0.0238 -0.9997>}
 spine_03 {parent=spine_02, loc=<6.0104 0 0>, rot=<0 0 0.0149 -0.9999>}
 spine_04 {parent=spine_03, loc=<8.1024 0 0>, rot=<0 0 -0.0132 -0.9999>}
 spine_05 {parent=spine_04, loc=<10.0146 0 0>, rot=<0 0 -0.0015 -1>}
 neck_01 {parent=spine_05, loc=<11.3138 0.1529 0>, rot=<0 0 -0.2003 -0.9797>}
 neck_02 {parent=neck_01, loc=<4.5627 0 0>, rot=<0 0 0.0504 -0.9987>}
 head {parent=neck_02, loc=<5.3914 0 0>, rot=<0 0 0.115 -0.9934>}
 attach_cape {parent=spine_05, loc=<9.017 -11.714 0>, rot=<0.0616 -0.7044 -0.0616 -0.7044>}
 cape {parent=attach_cape, type=Null, loc=<0 0 0>, rot=<0 0 0 1>}
 attach_backpack {parent=spine_05, loc=<-0.123 -13.473 0>, rot=<-0.0093 -0.707 0.0093 -0.707>}
 Backpack_BR {parent=attach_backpack, type=Null, loc=<0 1.6269 -0.1929>, rot=<-0.0692 0 0 0.9976>}
 clavicle_l {parent=spine_05, loc=<4.3575 -0.1466 -2.3223>, rot=<0.0465 -0.7289 -0.009 -0.6829>}
 upperarm_l {parent=clavicle_l, loc=<15.5111 0 0>, rot=<0.0226 -0.3234 -0.0088 -0.9459>}
 lowerarm_l {parent=upperarm_l, loc=<26.6514 0 0>, rot=<0.0391 -0.0099 -0.2453 -0.9686>}
 hand_l {parent=lowerarm_l, loc=<22.3491 0 0>, rot=<-0.7081 -0.0507 0.0975 -0.6975>}
 weapon_l {parent=hand_l, loc=<-0.1974 -3.4111 0.3811>, rot=<0 0 0 1>}
 thumb_01_l {parent=hand_l, loc=<1.3118 -0.3191 -2.459>, rot=<0.6546 -0.4042 0.0664 -0.6354>}
 thumb_02_l {parent=thumb_01_l, loc=<5.9336 -0.0448 0.0401>, rot=<-0.0037 0.0213 0.298 -0.9543>}
 thumb_03_l {parent=thumb_02_l, loc=<2.2223 0.3682 -0.0907>, rot=<0.0118 -0.013 0.0066 -0.9998>}
 index_metacarpal_l {parent=hand_l, loc=<1.6535 0.0183 -2.96>, rot=<0.0158 -0.0124 -0.0173 -0.9996>}
 index_01_l {parent=index_metacarpal_l, loc=<8.3008 0.0357 -0.0524>, rot=<-0.0044 0.0485 0.0924 -0.9945>}
 index_02_l {parent=index_01_l, loc=<3.3587 -0.3004 0.0452>, rot=<-0.002 0.0027 0.3225 -0.9466>}
 index_03_l {parent=index_02_l, loc=<2.5578 -0.123 -0.0252>, rot=<-0.002 0.0067 0.2558 -0.9667>}
 middle_metacarpal_l {parent=hand_l, loc=<2.1856 -0.1725 -0.5984>, rot=<0.006 0.0168 -0.0329 -0.9993>}
 middle_01_l {parent=middle_metacarpal_l, loc=<7.3275 -0.2447 0.0129>, rot=<-0.0208 0.0042 0.0858 -0.9961>}
 middle_02_l {parent=middle_01_l, loc=<4.0123 0.1401 0.1175>, rot=<-0.0218 0.0258 0.3512 -0.9357>}
 middle_03_l {parent=middle_02_l, loc=<2.6077 0.0534 0.1373>, rot=<0.0082 0.0188 0.1708 -0.9851>}
 ring_metacarpal_l {parent=hand_l, loc=<2.2074 -0.3625 1.3638>, rot=<-0.0284 0.0722 -0.0458 -0.9959>}
 ring_01_l {parent=ring_metacarpal_l, loc=<7.0295 -0.6197 -0.1446>, rot=<-0.0489 -0.0252 0.1892 -0.9804>}
 ring_02_l {parent=ring_01_l, loc=<3.6174 0.4647 -0.1027>, rot=<0 0 0.2806 -0.9598>}
 ring_03_l {parent=ring_02_l, loc=<2.4152 0.114 0.0199>, rot=<0 0 0.1189 -0.9929>}
 pinky_metacarpal_l {parent=hand_l, loc=<1.7616 -0.9753 3.3814>, rot=<-0.0397 0.0814 -0.069 -0.9935>}
 pinky_01_l {parent=pinky_metacarpal_l, loc=<7.1709 -0.4831 -0.1094>, rot=<-0.0654 -0.0035 0.2242 -0.9723>}
 pinky_02_l {parent=pinky_01_l, loc=<2.8274 0.0893 -0.1043>, rot=<0.0037 0.0011 0.2854 -0.9584>}
 pinky_03_l {parent=pinky_02_l, loc=<2.2046 0.0307 -0.0078>, rot=<-0.0002 0.0098 0.1445 -0.9895>}
 deform_wrist_dn_base_l {parent=hand_l, loc=<0 -2.2342 0>, rot=<0 0 0 1>}
 deform_wrist_dn_l {parent=deform_wrist_dn_base_l, loc=<0 0 0>, rot=<0 0 0 1>}
 deform_wrist_up_base_l {parent=hand_l, loc=<0 2.2682 0>, rot=<0 0 0 1>}
 deform_wrist_up_l {parent=deform_wrist_up_base_l, loc=<0 0 0>, rot=<0 0 0 1>}
 lowerarm_twist_02_l {parent=lowerarm_l, loc=<9.6199 0 0>, rot=<0 0 0 1>}
 lowerarm_twist_01_l {parent=lowerarm_l, loc=<17.8549 0 0>, rot=<0 0 0 1>}
 deform_elbow_in_base_l {parent=lowerarm_l, loc=<0.2112 5.259 0.8274>, rot=<0 0 0 1>}
 deform_elbow_in_l {parent=deform_elbow_in_base_l, loc=<0 0 0>, rot=<0 0 0 1>}
 deform_elbow_out_base_l {parent=lowerarm_l, loc=<0 -4 0>, rot=<0 0 0 1>}
 deform_elbow_out_l {parent=deform_elbow_out_base_l, loc=<0 0 0>, rot=<0 0 0 1>}
 upperarm_twist_01_l {parent=upperarm_l, loc=<0.0041 0 0>, rot=<0 0 0 1>}
 deform_upperarm_bk_base_l {parent=upperarm_twist_01_l, loc=<1.0813 -7.2193 -1.9453>, rot=<0 0 0 1>}
 deform_upperarm_bk_l {parent=deform_upperarm_bk_base_l, loc=<0 0 0>, rot=<0 0 0 1>}
 deform_upperarm_fr_base_l {parent=upperarm_twist_01_l, loc=<-0.3691 7.0506 -2.765>, rot=<0 0 0 1>}
 deform_upperarm_fr_l {parent=deform_upperarm_fr_base_l, loc=<0 0 0>, rot=<0 0 0 1>}
 deform_upperarm_up_base_l {parent=upperarm_twist_01_l, loc=<-3.2216 -1.1391 3.3311>, rot=<0 0 0 1>}
 deform_upperarm_up_l {parent=deform_upperarm_up_base_l, loc=<0 0 0>, rot=<0 0 0 1>}
 upperarm_twist_02_l {parent=upperarm_l, loc=<13.7941 0 0>, rot=<0 0 0 1>}
 upperarm_sleeves_l {parent=upperarm_l, loc=<0.0041 0 0>, rot=<0 0 0 1>}
 MSTR_sleeve_r_A_dyn {parent=upperarm_sleeves_l, loc=<5.6777 0.2613 1.6033>, rot=<0.1078 0.1137 0.031 -0.9872>}
 sleeve_r_A_dyn_1 {parent=MSTR_sleeve_r_A_dyn, loc=<0 0 0>, rot=<0 0 0 1>}
 sleeve_r_A_dyn_2 {parent=sleeve_r_A_dyn_1, loc=<12.0002 0.0001 -0.0005>, rot=<0 0 0 1>}
 MSTR_sleeve_r_B_dyn {parent=upperarm_sleeves_l, loc=<5.6636 1.9117 1.0748>, rot=<0.6215 0.076 -0.118 -0.7708>}
 sleeve_r_B_dyn_1 {parent=MSTR_sleeve_r_B_dyn, loc=<0 0 0>, rot=<0 0 0 1>}
 sleeve_r_B_dyn_2 {parent=sleeve_r_B_dyn_1, loc=<12.0001 0.0001 -0.0001>, rot=<0 0 0 1>}
 MSTR_sleeve_r_C_dyn {parent=upperarm_sleeves_l, loc=<5.5894 -2.1557 1.1442>, rot=<-0.6711 0.1157 0.1332 -0.7201>}
 sleeve_r_C_dyn_1 {parent=MSTR_sleeve_r_C_dyn, loc=<0 0 0>, rot=<0 0 0 1>}
 sleeve_r_C_dyn_2 {parent=sleeve_r_C_dyn_1, loc=<12.0001 -0.0001 0>, rot=<0 0 0 1>}
 MSTR_sleeve_r_D_dyn {parent=upperarm_sleeves_l, loc=<5.7427 2.4066 -3.2713>, rot=<0.4625 -0.0431 -0.1137 -0.8782>}
 sleeve_r_D_dyn_1 {parent=MSTR_sleeve_r_D_dyn, loc=<0 0 0>, rot=<0 0 0 1>}
 sleeve_r_D_dyn_2 {parent=sleeve_r_D_dyn_1, loc=<11.9994 -0.0004 0.0004>, rot=<0 0 0 1>}
 MSTR_sleeve_r_E_dyn {parent=upperarm_sleeves_l, loc=<5.6402 -2.4257 -3.8639>, rot=<0.961 -0.0723 -0.0772 0.2556>}
 sleeve_r_E_dyn_1 {parent=MSTR_sleeve_r_E_dyn, loc=<0 0 0>, rot=<0 0 0 1>}
 sleeve_r_E_dyn_2 {parent=sleeve_r_E_dyn_1, loc=<11.9999 0.0001 -0.0001>, rot=<0 0 0 1>}
 MSTR_necklase_l_dyn {parent=spine_05, loc=<6.6814 7.0713 -6.6438>, rot=<-0.2554 -0.731 -0.6239 -0.1061>}
 necklase_l_dyn_1 {parent=MSTR_necklase_l_dyn, loc=<0 0 0>, rot=<0 0 0 1>}
 necklase_l_dyn_2 {parent=necklase_l_dyn_1, loc=<5.685 0 0>, rot=<-0.0186 -0.115 -0.0256 -0.9929>}
 necklase_l_dyn_3 {parent=necklase_l_dyn_2, loc=<5.4 0 0>, rot=<-0.0155 -0.0486 -0.2044 -0.9776>}
 necklase_l_dyn_4 {parent=necklase_l_dyn_3, loc=<5.1358 0 0>, rot=<0.0301 -0.0478 -0.3826 -0.9222>}
 necklase_l_dyn_7 {parent=necklase_l_dyn_4, loc=<1.1175 0 0>, rot=<-0.053 -0.0578 0.6735 -0.735>}
 MSTR_necklase_r_dyn {parent=spine_05, loc=<6.7126 7.2659 6.7796>, rot=<0.0896 0.6614 0.7074 0.2325>}
 necklase_r_dyn_1 {parent=MSTR_necklase_r_dyn, loc=<0 0 0>, rot=<0 0 0 1>}
 necklase_r_dyn_2 {parent=necklase_r_dyn_1, loc=<5.6164 0 0>, rot=<0.024 -0.0587 0.0441 -0.997>}
 necklase_r_dyn_3 {parent=necklase_r_dyn_2, loc=<5.4 0 0>, rot=<0.0015 -0.0612 0.1738 -0.9829>}
 necklase_r_dyn_4 {parent=necklase_r_dyn_3, loc=<5.1491 0 0>, rot=<0.0266 -0.0442 0.381 -0.9231>}
 clavicle_r {parent=spine_05, loc=<4.3575 -0.1466 2.3223>, rot=<0.7289 0.0465 -0.6829 0.009>}
 upperarm_r {parent=clavicle_r, loc=<-15.5111 0 0>, rot=<0.0226 -0.3234 -0.0088 -0.9459>}
 lowerarm_r {parent=upperarm_r, loc=<-26.6515 0 0>, rot=<0.0391 -0.0099 -0.2453 -0.9686>}
 hand_r {parent=lowerarm_r, loc=<-22.3492 0 0>, rot=<-0.7081 -0.0507 0.0975 -0.6975>}
 weapon_r {parent=hand_r, loc=<0.1975 3.4112 -0.3811>, rot=<0 0 0 -1>}
 thumb_01_r {parent=hand_r, loc=<-1.3119 0.3192 2.459>, rot=<0.6546 -0.4042 0.0664 -0.6354>}
 thumb_02_r {parent=thumb_01_r, loc=<-5.9335 0.0447 -0.0401>, rot=<-0.0037 0.0213 0.298 -0.9543>}
 thumb_03_r {parent=thumb_02_r, loc=<-2.2223 -0.3682 0.0907>, rot=<0.0118 -0.013 0.0066 -0.9998>}
 index_metacarpal_r {parent=hand_r, loc=<-1.6534 -0.0181 2.9601>, rot=<0.0158 -0.0124 -0.0173 -0.9996>}
 index_01_r {parent=index_metacarpal_r, loc=<-8.3008 -0.0357 0.0524>, rot=<-0.0044 0.0485 0.0924 -0.9945>}
 index_02_r {parent=index_01_r, loc=<-3.3589 0.3003 -0.0453>, rot=<-0.002 0.0027 0.3225 -0.9466>}
 index_03_r {parent=index_02_r, loc=<-2.5578 0.1231 0.0251>, rot=<-0.002 0.0067 0.2558 -0.9667>}
 middle_metacarpal_r {parent=hand_r, loc=<-2.1856 0.1726 0.5984>, rot=<0.006 0.0168 -0.0329 -0.9993>}
 middle_01_r {parent=middle_metacarpal_r, loc=<-7.3273 0.2447 -0.0129>, rot=<-0.0208 0.0042 0.0858 -0.9961>}
 middle_02_r {parent=middle_01_r, loc=<-4.0142 -0.1411 -0.07>, rot=<-0.0217 0.0259 0.3512 -0.9357>}
 middle_03_r {parent=middle_02_r, loc=<-2.6074 -0.0534 -0.1312>, rot=<0.0083 0.0188 0.1708 -0.9851>}
 ring_metacarpal_r {parent=hand_r, loc=<-2.2073 0.3627 -1.3638>, rot=<-0.0284 0.0722 -0.0458 -0.9959>}
 ring_01_r {parent=ring_metacarpal_r, loc=<-7.0297 0.6197 0.1446>, rot=<-0.0489 -0.0252 0.1892 -0.9804>}
 ring_02_r {parent=ring_01_r, loc=<-3.6171 -0.4648 0.1027>, rot=<0 0 0.2806 -0.9598>}
 ring_03_r {parent=ring_02_r, loc=<-2.4153 -0.1141 -0.0199>, rot=<0 0 0.1189 -0.9929>}
 pinky_metacarpal_r {parent=hand_r, loc=<-1.7617 0.9753 -3.3814>, rot=<-0.0397 0.0814 -0.069 -0.9935>}
 pinky_01_r {parent=pinky_metacarpal_r, loc=<-7.1707 0.4832 0.1094>, rot=<-0.0654 -0.0035 0.2242 -0.9723>}
 pinky_02_r {parent=pinky_01_r, loc=<-2.8274 -0.0892 0.1043>, rot=<0.0037 0.0011 0.2854 -0.9584>}
 pinky_03_r {parent=pinky_02_r, loc=<-2.2045 -0.031 0.0078>, rot=<-0.0002 0.0098 0.1445 -0.9895>}
 deform_wrist_dn_base_r {parent=hand_r, loc=<-0.0003 2.2346 -0.0001>, rot=<0 0 0 1>}
 deform_wrist_dn_r {parent=deform_wrist_dn_base_r, loc=<0 0 0>, rot=<0 0 0 1>}
 deform_wrist_up_base_r {parent=hand_r, loc=<0.0004 -2.2683 0.0001>, rot=<0 0 0 1>}
 deform_wrist_up_r {parent=deform_wrist_up_base_r, loc=<0 0 0>, rot=<0 0 0 1>}
 lowerarm_twist_02_r {parent=lowerarm_r, loc=<-9.6199 0 0>, rot=<0 0 0 1>}
 lowerarm_twist_01_r {parent=lowerarm_r, loc=<-17.855 0 0>, rot=<0 0 0 1>}
 watch_up_deform_base {parent=lowerarm_twist_01_r, loc=<-2.1306 0 -2.9392>, rot=<0 -0.0547 0 -0.9985>}
 watch_up_deform {parent=watch_up_deform_base, loc=<0 0 0>, rot=<0 0 0 1>}
 watch_down_deform_base {parent=lowerarm_twist_01_r, loc=<-1.6024 0 1.8704>, rot=<0 0.0044 0 -1>}
 watch_down_deform {parent=watch_down_deform_base, loc=<0 0 0>, rot=<0 0 0 1>}
 deform_elbow_in_base_r {parent=lowerarm_r, loc=<-0.2111 -5.2591 -0.8275>, rot=<0 0 0 1>}
 deform_elbow_in_r {parent=deform_elbow_in_base_r, loc=<-0.0001 0 0.0002>, rot=<0 0 0 1>}
 deform_elbow_out_base_r {parent=lowerarm_r, loc=<0 4 0>, rot=<0 0 0 1>}
 deform_elbow_out_r {parent=deform_elbow_out_base_r, loc=<0 0 0>, rot=<0 0 0 1>}
 upperarm_twist_01_r {parent=upperarm_r, loc=<-0.0041 0 0>, rot=<0 0 0 1>}
 deform_upperarm_bk_base_r {parent=upperarm_twist_01_r, loc=<-1.0815 7.2192 1.9456>, rot=<0 0 0 1>}
 deform_upperarm_bk_r {parent=deform_upperarm_bk_base_r, loc=<0 0 0>, rot=<0 0 0 1>}
 deform_upperarm_fr_base_r {parent=upperarm_twist_01_r, loc=<0.3689 -7.0506 2.7652>, rot=<0 0 0 1>}
 deform_upperarm_fr_r {parent=deform_upperarm_fr_base_r, loc=<0 0 0>, rot=<0 0 0 1>}
 deform_upperarm_up_base_r {parent=upperarm_twist_01_r, loc=<3.2213 1.1391 -3.3308>, rot=<0 0 0 1>}
 deform_upperarm_up_r {parent=deform_upperarm_up_base_r, loc=<0 0 0>, rot=<0 0 0 -1>}
 upperarm_twist_02_r {parent=upperarm_r, loc=<-13.7942 0 0>, rot=<0 0 0 1>}
 upperarm_sleeves_r {parent=upperarm_r, loc=<-0.0041 0 0>, rot=<0 0 0 1>}
 MSTR_sleeve_l_A_dyn {parent=upperarm_sleeves_r, loc=<-5.6781 -0.2612 -1.603>, rot=<-0.031 -0.9872 0.1078 -0.1137>}
 sleeve_l_A_dyn_1 {parent=MSTR_sleeve_l_A_dyn, loc=<0 0 0>, rot=<0 0 0 1>}
 sleeve_l_A_dyn_2 {parent=sleeve_l_A_dyn_1, loc=<12 0 0>, rot=<0 0 0 1>}
 MSTR_sleeve_l_B_dyn {parent=upperarm_sleeves_r, loc=<-5.6635 -1.9117 -1.0748>, rot=<0.118 -0.7708 0.6215 -0.076>}
 sleeve_l_B_dyn_1 {parent=MSTR_sleeve_l_B_dyn, loc=<0 0 0>, rot=<0 0 0 1>}
 sleeve_l_B_dyn_2 {parent=sleeve_l_B_dyn_1, loc=<12 0 0>, rot=<0 0 0 1>}
 MSTR_sleeve_l_C_dyn {parent=upperarm_sleeves_r, loc=<-5.5894 2.1557 -1.1443>, rot=<-0.1332 -0.7201 -0.6711 -0.1157>}
 sleeve_l_C_dyn_1 {parent=MSTR_sleeve_l_C_dyn, loc=<0 0 0>, rot=<0 0 0 1>}
 sleeve_l_C_dyn_2 {parent=sleeve_l_C_dyn_1, loc=<12 0 0>, rot=<0 0 0 1>}
 MSTR_sleeve_l_D_dyn {parent=upperarm_sleeves_r, loc=<-5.7425 -2.4065 3.2709>, rot=<0.1137 -0.8782 0.4625 0.0431>}
 sleeve_l_D_dyn_1 {parent=MSTR_sleeve_l_D_dyn, loc=<0 0 0>, rot=<0 0 0 1>}
 sleeve_l_D_dyn_2 {parent=sleeve_l_D_dyn_1, loc=<12 0 0>, rot=<0 0 0 1>}
 MSTR_sleeve_l_E_dyn {parent=upperarm_sleeves_r, loc=<-5.6401 2.4257 3.8637>, rot=<0.0772 0.2556 0.961 0.0723>}
 sleeve_l_E_dyn_1 {parent=MSTR_sleeve_l_E_dyn, loc=<0 0 0>, rot=<0 0 0 1>}
 sleeve_l_E_dyn_2 {parent=sleeve_l_E_dyn_1, loc=<12 0 0>, rot=<0 0 0 1>}
 collar_back_deform_base {parent=spine_05, loc=<13.3899 -4.9041 0>, rot=<0.7061 0.0382 -0.0382 -0.7061>}
 collar_back_deform {parent=collar_back_deform_base, loc=<0 0 0>, rot=<0 0 0 1>}
 collar_back_r_A_deform_base {parent=spine_05, loc=<13.2592 -2.4519 5.865>, rot=<0.9225 0.0499 -0.0207 -0.3821>}
 collar_back_r_A_deform {parent=collar_back_r_A_deform_base, loc=<0 0 0>, rot=<0 0 0 1>}
 collar_back_r_B_deform_base {parent=spine_05, loc=<12.7504 0.2337 6.2956>, rot=<0.9985 0.054 0 0>}
 collar_back_r_B_deform {parent=collar_back_r_B_deform_base, loc=<0 0 0>, rot=<0 0 0 1>}
 collar_back_l_A_deform_base {parent=spine_05, loc=<13.2591 -2.452 -5.865>, rot=<0.3821 0.0207 -0.0499 -0.9225>}
 collar_back_l_A_deform {parent=collar_back_l_A_deform_base, loc=<0 0 0>, rot=<0 0 0 1>}
 collar_back_l_B_deform_base {parent=spine_05, loc=<12.7503 0.2337 -6.2956>, rot=<0 0 -0.054 -0.9985>}
 collar_back_l_B_deform {parent=collar_back_l_B_deform_base, loc=<0 0 0>, rot=<0 0 0 1>}
 deform_pec_aim_base_l {parent=spine_04, loc=<7.8732 12.6388 -2.4123>, rot=<0.0102 -0.4602 0.3979 -0.7936>}
 deform_pec_aim_l {parent=deform_pec_aim_base_l, loc=<0 0 0>, rot=<0 0 0 1>}
 deform_pec_base_l {parent=deform_pec_aim_l, loc=<10.766 0 0>, rot=<0 0 0 1>}
 deform_pec_l {parent=deform_pec_base_l, loc=<0.055 0 0>, rot=<0 0 0 1>}
 deform_pec_aim_base_r {parent=spine_04, loc=<7.8727 12.6388 2.4123>, rot=<-0.4602 -0.0102 0.7936 0.3979>}
 deform_pec_aim_r {parent=deform_pec_aim_base_r, loc=<0 0 0>, rot=<0 0 0 1>}
 deform_pec_base_r {parent=deform_pec_aim_r, loc=<-10.7662 -0.0006 -0.0007>, rot=<0 0 0 1>}
 deform_pec_r {parent=deform_pec_base_r, loc=<0 0 0>, rot=<0 0 0 1>}
 MSTR_tshirt_mid_l_dyn_1 {parent=spine_01, loc=<1.6532 -12.4541 -0.8622>, rot=<0.0228 -0.7067 0.7067 -0.0228>}
 tshirt_mid_l_dyn_1 {parent=MSTR_tshirt_mid_l_dyn_1, loc=<0 0 0>, rot=<0 0 0 1>}
 tshirt_mid_l_dyn_2 {parent=tshirt_mid_l_dyn_1, loc=<6.8 0 0>, rot=<0 0 0 1>}
 tshirt_mid_l_dyn_3 {parent=tshirt_mid_l_dyn_2, loc=<7.5 0 0>, rot=<0 0 0 1>}
 MSTR_tshirt_A_r_dyn_1 {parent=spine_01, loc=<1.7838 -10.2477 -12.6346>, rot=<0.0181 -0.9068 0.4207 -0.0196>}
 tshirt_A_r_dyn_1 {parent=MSTR_tshirt_A_r_dyn_1, loc=<0 0 0>, rot=<0 0 0 1>}
 tshirt_A_r_dyn_2 {parent=tshirt_A_r_dyn_1, loc=<6.8 0 0>, rot=<0 0 0 1>}
 tshirt_A_r_dyn_3 {parent=tshirt_A_r_dyn_2, loc=<7.5 0 0>, rot=<0 0 0 1>}
 MSTR_tshirt_A_l_dyn_1 {parent=spine_01, loc=<1.601 0.3455 -16.6409>, rot=<0.0161 -0.9996 -0.0004 -0.0223>}
 tshirt_A_l_dyn_1 {parent=MSTR_tshirt_A_l_dyn_1, loc=<0 0 0>, rot=<0 0 0 1>}
 tshirt_A_l_dyn_2 {parent=tshirt_A_l_dyn_1, loc=<6.8 0 0>, rot=<0 0 0 1>}
 tshirt_A_l_dyn_3 {parent=tshirt_A_l_dyn_2, loc=<7.5 0 0>, rot=<0 0 0 1>}
 MSTR_tshirt_A_l_dyn_4 {parent=spine_01, loc=<1.7833 -10.2477 12.6346>, rot=<0.0196 -0.4207 0.9068 -0.0181>}
 tshirt_A_l_dyn_4 {parent=MSTR_tshirt_A_l_dyn_4, loc=<0 0 0>, rot=<0 0 0 1>}
 tshirt_A_l_dyn_21 {parent=tshirt_A_l_dyn_4, loc=<6.7995 -0.0001 0.0001>, rot=<0 0 0 1>}
 tshirt_A_l_dyn_31 {parent=tshirt_A_l_dyn_21, loc=<7.5 0 -0.0001>, rot=<0 0 0 1>}
 MSTR_tshirt_A_l_dyn_8 {parent=spine_01, loc=<1.6006 0.3455 16.6409>, rot=<0.0223 0.0004 0.9996 -0.0161>}
 tshirt_A_l_dyn_8 {parent=MSTR_tshirt_A_l_dyn_8, loc=<0 0 0>, rot=<0 0 0 1>}
 tshirt_A_l_dyn_22 {parent=tshirt_A_l_dyn_8, loc=<6.7996 0 0.0001>, rot=<0 0 0 -1>}
 tshirt_A_l_dyn_32 {parent=tshirt_A_l_dyn_22, loc=<7.5001 0 -0.0001>, rot=<0 0 0 -1>}
 tshirt_A_l_dyn_pnt {parent=tshirt_A_l_dyn_32, loc=<1.79 -6.23 -1.56>, rot=<0 0 0 1>}
 thigh_l {parent=pelvis, loc=<-2.5693 -0.0109 -7.7501>, rot=<0.0012 0.0572 0.0155 -0.9982>}
 calf_l {parent=thigh_l, loc=<-39.0484 0 0>, rot=<0.0006 -0.0161 -0.0391 -0.9991>}
 foot_l {parent=calf_l, loc=<-42.561 0.1077 0.4292>, rot=<0.0418 -0.0455 0.0343 -0.9975>}
 ball_l {parent=foot_l, loc=<-6.1599 13.8264 0.6806>, rot=<0.0001 -0.0001 -0.7063 -0.7079>}
 foot_l_deform_base {parent=foot_l, loc=<0.5041 -0.0921 0.1615>, rot=<0.5234 0.4729 0.5208 0.4807>}
 foot_l_deform_for_base {parent=foot_l_deform_base, loc=<4 0 1>, rot=<0 0 0 1>}
 foot_l_deform_for {parent=foot_l_deform_for_base, loc=<0 0 0>, rot=<0 0 0 1>}
 foot_l_deform_back_base {parent=foot_l_deform_base, loc=<-3.5 0 0>, rot=<0 -1 0 0>}
 foot_l_deform_back {parent=foot_l_deform_back_base, loc=<0 0 0>, rot=<0 0 0 1>}
 calf_twist_02_l {parent=calf_l, loc=<-18.8064 -0.0967 0.0397>, rot=<0.0237 0.0095 -0.0035 -0.9997>}
 calf_twist_01_l {parent=calf_l, loc=<-35.813 -0.0967 0.0397>, rot=<0.0237 0.0095 -0.0035 -0.9997>}
 deform_knee_in_base_l {parent=calf_l, loc=<1.298 -5.2957 -0.0496>, rot=<0 0 0 1>}
 deform_knee_in_l {parent=deform_knee_in_base_l, loc=<0 0 0>, rot=<0 0 0 1>}
 deform_knee_out_base_l {parent=calf_l, loc=<0 4 0>, rot=<0 0 0 1>}
 deform_knee_out_l {parent=deform_knee_out_base_l, loc=<0 0 0>, rot=<0 0 0 1>}
 thigh_twist_01_l {parent=thigh_l, loc=<-21.6303 0 0>, rot=<0.0013 0 0 -1>}
 thigh_twist_01_for_base_l {parent=thigh_twist_01_l, loc=<21 3 0>, rot=<0.1142 -0.0032 0.9935 0>}
 thigh_twist_01_for_l {parent=thigh_twist_01_for_base_l, loc=<0 0 0>, rot=<0 0 0 1>}
 deform_groin_base_l {parent=thigh_l, loc=<-2.1908 8.1142 -1.7442>, rot=<0 0 0 1>}
 deform_groin_l {parent=deform_groin_base_l, loc=<0 0 0>, rot=<0 0 0 1>}
 thigh_r {parent=pelvis, loc=<-2.5691 -0.0109 7.7501>, rot=<0.0572 -0.0012 0.9982 0.0155>}
 calf_r {parent=thigh_r, loc=<39.0487 0 0>, rot=<0.0006 -0.0161 -0.0391 -0.9991>}
 foot_r {parent=calf_r, loc=<42.5612 -0.1077 -0.4292>, rot=<0.0418 -0.0455 0.0343 -0.9975>}
 ball_r {parent=foot_r, loc=<6.1598 -13.8264 -0.6806>, rot=<0.0001 -0.0001 -0.7063 -0.7079>}
 foot_r_deform_base {parent=foot_r, loc=<-0.5042 0.092 -0.1615>, rot=<-0.5208 0.4807 0.5234 -0.4729>}
 foot_r_deform_for_base {parent=foot_r_deform_base, loc=<4 0 1>, rot=<0 0 0 1>}
 foot_r_deform_for {parent=foot_r_deform_for_base, loc=<0 0 0>, rot=<0 0 0 1>}
 foot_r_deform_back_base {parent=foot_r_deform_base, loc=<-3.5 0 0>, rot=<0 -1 0 0>}
 foot_r_deform_back {parent=foot_r_deform_back_base, loc=<0 0 0>, rot=<0 0 0 1>}
 calf_twist_02_r {parent=calf_r, loc=<18.8064 0.0967 -0.0397>, rot=<0.0237 0.0095 -0.0035 -0.9997>}
 calf_twist_01_r {parent=calf_r, loc=<35.8132 0.0967 -0.0397>, rot=<0.0237 0.0095 -0.0035 -0.9997>}
 deform_knee_in_base_r {parent=calf_r, loc=<-1.2981 5.2958 0.0496>, rot=<0 0 0 1>}
 deform_knee_in_r {parent=deform_knee_in_base_r, loc=<0 0 0>, rot=<0 0 0 1>}
 deform_knee_out_base_r {parent=calf_r, loc=<0 -4 0>, rot=<0 0 0 1>}
 deform_knee_out_r {parent=deform_knee_out_base_r, loc=<0 0 0>, rot=<0 0 0 1>}
 thigh_twist_01_r {parent=thigh_r, loc=<21.6303 0 0>, rot=<0.0013 0 0 -1>}
 thigh_twist_01_for_base_r {parent=thigh_twist_01_r, loc=<-20.9998 -3 0>, rot=<0 0 0 1>}
 thigh_twist_01_for_r {parent=thigh_twist_01_for_base_r, loc=<0 0 0>, rot=<0 0 0 1>}
 MSTR_chain_leg_A_1 {parent=thigh_twist_01_r, loc=<-10 0 0>, rot=<-0.1097 0.6986 0.1097 -0.6986>}
 chain_leg_A_1 {parent=MSTR_chain_leg_A_1, loc=<0 0 0>, rot=<0 0 0 1>}
 chain_leg_A_2 {parent=chain_leg_A_1, loc=<8 0 0>, rot=<0 0 0 1>}
 MSTR_chain_leg_B_1 {parent=thigh_twist_01_r, loc=<-11 0 0>, rot=<-0.1097 0.6986 0.1097 -0.6986>}
 chain_leg_B_1 {parent=MSTR_chain_leg_B_1, loc=<0 0 0>, rot=<0 0 0 1>}
 chain_leg_B_2 {parent=chain_leg_B_1, loc=<8 0 0>, rot=<0 0 0 1>}
 twist_01_r_pnt {parent=thigh_twist_01_r, loc=<-9.78 -2.4 6.67>, rot=<0 0 0 1>}
 deform_groin_base_r {parent=thigh_r, loc=<2.1909 -8.1142 1.7442>, rot=<0 0 0 1>}
 deform_groin_r {parent=deform_groin_base_r, loc=<0 0 0>, rot=<0 0 0 1>}
 deform_glute_base_l {parent=pelvis, loc=<-8.8064 -8.8222 -8.4287>, rot=<0.0275 0.0501 -0.4472 -0.8926>}
 deform_glute_l {parent=deform_glute_base_l, loc=<0 0 0>, rot=<0 0 0 1>}
 deform_glute_base_r {parent=pelvis, loc=<-8.8062 -8.8222 8.4287>, rot=<0.0501 -0.0275 0.8926 -0.4472>}
 deform_glute_r {parent=deform_glute_base_r, loc=<0 0 0>, rot=<0 0 0 1>}
 deform_glute_base_m {parent=pelvis, loc=<-6.6047 -5.731 0>, rot=<-0.0322 -0.7064 0.7064 0.0322>}
 deform_glute_m {parent=deform_glute_base_m, loc=<0 0 0>, rot=<0 0 0 1>}
 thigh_r_aim {parent=pelvis, loc=<-2.5691 -0.0109 7.7501>, rot=<0.0572 -0.0012 0.9982 0.0155>}
 thigh_r_aim_vec {parent=thigh_r_aim, loc=<5 0 0>, rot=<0 0 0 1>}
 thigh_r_aim_up {parent=pelvis, loc=<7.3945 -0.3256 15.6638>, rot=<0.0572 -0.0012 0.9982 0.0155>}
 attach {parent=root, loc=<0 0 0>, rot=<0 0 0 1>}
 ik_hand_root {parent=root, loc=<0 0 0>, rot=<0 0 0 1>}
 ik_hand_gun {parent=ik_hand_root, loc=<-51.9729 4.0884 102.6935>, rot=<0.6084 0.1793 0.4613 -0.6205>}
 ik_hand_r {parent=ik_hand_gun, loc=<0 0 0>, rot=<0 0 0 1>}
 ik_hand_l {parent=ik_hand_gun, loc=<53.0286 82.1782 35.2051>, rot=<-0.5102 -0.7906 -0.3387 0>}
 ik_foot_root {parent=root, loc=<0 0 0>, rot=<0 0 0 1>}
 ik_foot_l {parent=ik_foot_root, loc=<15.2608 -4.4176 9.4433>, rot=<0.0357 0.7045 0.0284 -0.7082>}
 ik_foot_r {parent=ik_foot_root, loc=<-15.2609 -4.4176 9.4431>, rot=<0.7082 0.0284 -0.7045 0.0357>}
]
variables [IsFrontend {type=bool, default=True}]
