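# Physics asset for the necklace on the M_MED_FunkyFeather_Stripes player body, defining its capsule bodies and joint limits.
# Left and right bone chains start at MSTR anchor bodies, with limited swing and tight twist at each joint.
# Both chains constrain to necklase_l_dyn_4, closing a loop from which the larger necklase_l_dyn_7 capsule continues like a pendant.
# Unconstrained capsules on spine_04, spine_05 and upperarm_r provide torso and right-arm collision shapes.
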PhysicsAsset /BRCosmetics/Characters/Player/Male/Medium/Bodies/M_MED_FunkyFeather_Stripes/Meshes/M_MED_FunkyFeather_Stripes_Necklase_Physics
bodies {
 MSTR_necklase_l_dyn [{shapeType=Capsule, length=1, radius=1, shapeName=MSTR_necklase_l_dyn_capsule}]
 MSTR_necklase_r_dyn [{shapeType=Capsule, length=1, radius=1, shapeName=MSTR_necklase_r_dyn_capsule}]
 necklase_l_dyn_1 [{shapeType=Capsule, length=3.1385, radius=1, shapeName=necklase_l_dyn_1_capsule, center=<2.9678 0 0>, rotation=<rot 90 0 0>}]
 necklase_l_dyn_2 [{shapeType=Capsule, length=2.9806, radius=1, shapeName=necklase_l_dyn_2_capsule, center=<2.6945 0 0>, rotation=<rot 90 0 0>}]
 necklase_l_dyn_3 [{shapeType=Capsule, length=2.9661, radius=1, shapeName=necklase_l_dyn_3_capsule, center=<2.4183 0 0>, rotation=<rot 90 0 0>}]
 necklase_l_dyn_4 [{shapeType=Capsule, length=0.0204, radius=1, shapeName=necklase_l_dyn_4_capsule, center=<0.9981 0 0>, rotation=<rot 90 0 0>}]
 necklase_l_dyn_7 [{
   shapeType=Capsule
   length=2.2081
   radius=3.7243
   shapeName=necklase_l_dyn_7_capsule
   center=<5.1316 0 2.5604>
   rotation=<rot 85.8005 180 180>
  }]
 necklase_r_dyn_1 [{shapeType=Capsule, length=3.2957, radius=1, shapeName=necklase_r_dyn_1_capsule, center=<2.9678 0 0>, rotation=<rot 90 0 0>}]
 necklase_r_dyn_2 [{shapeType=Capsule, length=2.9681, radius=1, shapeName=necklase_r_dyn_2_capsule, center=<2.7236 0 0>, rotation=<rot 90 0 0>}]
 necklase_r_dyn_3 [{shapeType=Capsule, length=2.6235, radius=1.0314, shapeName=necklase_r_dyn_3_capsule, center=<2.6435 0 0.1832>, rotation=<rot 90 0 0>}]
 spine_04 [{shapeType=Capsule, length=8.818, radius=20.258, shapeName=spine_04_capsule, center=<-0.8342 -7.7995 0>, rotation=<rot -90 -5.1219 0>}]
 spine_05 [{shapeType=Capsule, length=6.7357, radius=3.0115, shapeName=spine_05_capsule, center=<2.0939 8.5492 0>, rotation=<rot 0 0 90>}]
 upperarm_r [{shapeType=Capsule, length=6.1003, radius=7.0585, shapeName=upperarm_r_capsule, center=<-9.4022 -0.0012 1.8524>, rotation=<rot 90 0.0073 0>}]
}
constraints [
 {bone1Name=necklase_l_dyn_1, bone2Name=MSTR_necklase_l_dyn, swing1LimitDegrees=10, swing2LimitDegrees=25, twistLimitDegrees=5}
 {bone1Name=necklase_l_dyn_2, bone2Name=necklase_l_dyn_1, swing1LimitDegrees=30, swing2LimitDegrees=30, twistLimitDegrees=5}
 {bone1Name=necklase_r_dyn_1, bone2Name=MSTR_necklase_r_dyn, swing1LimitDegrees=30, swing2LimitDegrees=45, twistLimitDegrees=5}
 {bone1Name=necklase_r_dyn_2, bone2Name=necklase_r_dyn_1, swing1LimitDegrees=30, swing2LimitDegrees=30, twistLimitDegrees=5}
 {bone1Name=necklase_l_dyn_3, bone2Name=necklase_l_dyn_2, swing1LimitDegrees=30, swing2LimitDegrees=30, twistLimitDegrees=5}
 {bone1Name=necklase_r_dyn_3, bone2Name=necklase_r_dyn_2, swing1LimitDegrees=30, swing2LimitDegrees=30, twistLimitDegrees=5}
 {bone1Name=necklase_l_dyn_4, bone2Name=necklase_l_dyn_3, swing1LimitDegrees=30, swing2LimitDegrees=30, twistLimitDegrees=5}
 {bone1Name=necklase_l_dyn_4, bone2Name=necklase_r_dyn_3, swing1LimitDegrees=30, swing2LimitDegrees=30, twistLimitDegrees=5}
 {bone1Name=necklase_l_dyn_7, bone2Name=necklase_l_dyn_4, swing1LimitDegrees=30, swing2LimitDegrees=30, twistLimitDegrees=5}
]
