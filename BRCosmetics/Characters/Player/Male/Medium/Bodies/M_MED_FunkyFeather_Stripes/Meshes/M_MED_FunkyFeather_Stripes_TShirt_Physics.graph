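# Physics asset for the T-shirt of the medium male FunkyFeather Stripes player body, rigging its dynamic cloth bones with capsule bodies.
# Five three-bone chains each hang from a small MSTR_ anchor body, with tight swing and twist limits, alongside larger pelvis and thigh capsules.
# Angularly free constraints link the five chain tips side by side and tie the two end tips to the left and right thighs.
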PhysicsAsset /BRCosmetics/Characters/Player/Male/Medium/Bodies/M_MED_FunkyFeather_Stripes/Meshes/M_MED_FunkyFeather_Stripes_TShirt_Physics
bodies {
 MSTR_tshirt_A_l_dyn_1 [{shapeType=Capsule, length=1, radius=1, shapeName=MSTR_tshirt_A_l_dyn_1_capsule}]
 MSTR_tshirt_A_l_dyn_4 [{shapeType=Capsule, length=1, radius=1, shapeName=MSTR_tshirt_A_l_dyn_4_capsule}]
 MSTR_tshirt_A_l_dyn_8 [{shapeType=Capsule, length=1, radius=1, shapeName=MSTR_tshirt_A_l_dyn_8_capsule}]
 MSTR_tshirt_A_r_dyn_1 [{shapeType=Capsule, length=1, radius=1, shapeName=MSTR_tshirt_A_r_dyn_1_capsule}]
 MSTR_tshirt_mid_l_dyn_1 [{shapeType=Capsule, length=1, radius=1, shapeName=MSTR_tshirt_mid_l_dyn_1_capsule}]
 pelvis [{shapeType=Capsule, length=8.1847, radius=12.2157, shapeName=pelvis_capsule}]
 thigh_l [{shapeType=Capsule, length=10.1838, radius=8, shapeName=thigh_l_capsule, center=<-12 0 0>, rotation=<rot 90 0 0>}]
 thigh_r [{shapeType=Capsule, length=10.1838, radius=9, shapeName=thigh_r_capsule, center=<12 0 0>, rotation=<rot 90 0 0>}]
 tshirt_A_l_dyn_1 [{shapeType=Capsule, length=0.1152, radius=5.9676, shapeName=tshirt_A_l_dyn_1_capsule, center=<3.5684 0 4.6644>, rotation=<rot 90 0 0>}]
 tshirt_A_l_dyn_2 [{shapeType=Capsule, length=0.1152, radius=5.9676, shapeName=tshirt_A_l_dyn_2_capsule, center=<3.5684 0 4.6644>, rotation=<rot 90 0 0>}]
 tshirt_A_l_dyn_21 [{shapeType=Capsule, length=0.1152, radius=4.8, shapeName=tshirt_A_l_dyn_21_capsule, center=<3.5684 0 4.6644>, rotation=<rot 90 0 0>}]
 tshirt_A_l_dyn_22 [{shapeType=Capsule, length=0.1152, radius=5.9676, shapeName=tshirt_A_l_dyn_22_capsule, center=<3.5684 0 4.6644>, rotation=<rot 90 0 0>}]
 tshirt_A_l_dyn_3 [{shapeType=Capsule, length=0.1152, radius=5.9676, shapeName=tshirt_A_l_dyn_3_capsule, center=<3.5684 0 4.6644>, rotation=<rot 90 0 0>}]
 tshirt_A_l_dyn_31 [{shapeType=Capsule, length=0.1152, radius=4.8, shapeName=tshirt_A_l_dyn_31_capsule, center=<3.5684 0 4.6644>, rotation=<rot 90 0 0>}]
 tshirt_A_l_dyn_32 [{shapeType=Capsule, length=0.1152, radius=6.2, shapeName=tshirt_A_l_dyn_32_capsule, center=<3.5684 0 4.6644>, rotation=<rot 90 0 0>}]
 tshirt_A_l_dyn_4 [{shapeType=Capsule, length=0.1152, radius=5.2, shapeName=tshirt_A_l_dyn_4_capsule, center=<3.5684 0 4.6644>, rotation=<rot 90 0 0>}]
 tshirt_A_l_dyn_8 [{shapeType=Capsule, length=0.1152, radius=5.9676, shapeName=tshirt_A_l_dyn_8_capsule, center=<3.5684 0 4.6644>, rotation=<rot 90 0 0>}]
 tshirt_A_r_dyn_1 [{shapeType=Capsule, length=1.1334, radius=4.8, shapeName=tshirt_A_r_dyn_1_capsule, center=<3.5684 0 4.6644>, rotation=<rot 90 0 0>}]
 tshirt_A_r_dyn_2 [{shapeType=Capsule, length=1.1334, radius=4.8, shapeName=tshirt_A_r_dyn_2_capsule, center=<3.5684 0 4.6644>, rotation=<rot 90 0 0>}]
 tshirt_A_r_dyn_3 [{shapeType=Capsule, length=1.1334, radius=4.8, shapeName=tshirt_A_r_dyn_3_capsule, center=<3.5684 0 4.6644>, rotation=<rot 90 0 0>}]
 tshirt_mid_l_dyn_1 [{shapeType=Capsule, length=0.1152, radius=5.9676, shapeName=tshirt_mid_l_dyn_1_capsule, center=<3.5684 0 4.6644>, rotation=<rot 90 0 0>}]
 tshirt_mid_l_dyn_2 [{shapeType=Capsule, length=0.1152, radius=5.9676, shapeName=tshirt_mid_l_dyn_2_capsule, center=<3.5684 0 4.6644>, rotation=<rot 90 0 0>}]
 tshirt_mid_l_dyn_3 [{shapeType=Capsule, length=0.1152, radius=5.9676, shapeName=tshirt_mid_l_dyn_3_capsule, center=<3.5684 0 4.6644>, rotation=<rot 90 0 0>}]
}
constraints [
 {bone1Name=tshirt_mid_l_dyn_3, bone2Name=tshirt_mid_l_dyn_2, swing1LimitDegrees=1, swing2LimitDegrees=15, twistLimitDegrees=1}
 {bone1Name=tshirt_mid_l_dyn_2, bone2Name=tshirt_mid_l_dyn_1, swing1LimitDegrees=1, swing2LimitDegrees=15, twistLimitDegrees=1}
 {bone1Name=tshirt_mid_l_dyn_1, bone2Name=MSTR_tshirt_mid_l_dyn_1, swing1LimitDegrees=1, swing2LimitDegrees=15, twistLimitDegrees=1}
 {bone1Name=tshirt_A_r_dyn_3, bone2Name=tshirt_A_r_dyn_2, swing1LimitDegrees=1, swing2LimitDegrees=15, twistLimitDegrees=1}
 {bone1Name=tshirt_A_r_dyn_2, bone2Name=tshirt_A_r_dyn_1, swing1LimitDegrees=1, swing2LimitDegrees=15, twistLimitDegrees=1}
 {bone1Name=tshirt_A_r_dyn_1, bone2Name=MSTR_tshirt_A_r_dyn_1, swing1LimitDegrees=1, swing2LimitDegrees=15, twistLimitDegrees=1}
 {bone1Name=tshirt_A_l_dyn_32, bone2Name=tshirt_A_l_dyn_22, swing1LimitDegrees=1, swing2LimitDegrees=15, twistLimitDegrees=1}
 {bone1Name=tshirt_A_l_dyn_22, bone2Name=tshirt_A_l_dyn_8, swing1LimitDegrees=1, swing2LimitDegrees=15, twistLimitDegrees=1}
 {bone1Name=tshirt_A_l_dyn_8, bone2Name=MSTR_tshirt_A_l_dyn_8, swing1LimitDegrees=10, swing2LimitDegrees=15, twistLimitDegrees=1}
 {bone1Name=tshirt_A_l_dyn_31, bone2Name=tshirt_A_l_dyn_21, swing1LimitDegrees=1, swing2LimitDegrees=15, twistLimitDegrees=1}
 {bone1Name=tshirt_A_l_dyn_21, bone2Name=tshirt_A_l_dyn_4, swing1LimitDegrees=1, swing2LimitDegrees=15, twistLimitDegrees=1}
 {bone1Name=tshirt_A_l_dyn_4, bone2Name=MSTR_tshirt_A_l_dyn_4, swing1LimitDegrees=1, swing2LimitDegrees=15, twistLimitDegrees=1}
 {bone1Name=tshirt_A_l_dyn_3, bone2Name=tshirt_A_l_dyn_2, swing1LimitDegrees=1, swing2LimitDegrees=15, twistLimitDegrees=1}
 {bone1Name=tshirt_A_l_dyn_2, bone2Name=tshirt_A_l_dyn_1, swing1LimitDegrees=1, swing2LimitDegrees=15, twistLimitDegrees=1}
 {bone1Name=tshirt_A_l_dyn_1, bone2Name=MSTR_tshirt_A_l_dyn_1, swing1LimitDegrees=10, swing2LimitDegrees=15, twistLimitDegrees=1}
 {
  bone1Name=tshirt_A_l_dyn_32
  bone2Name=thigh_r
  swing1LimitDegrees=45
  swing1Motion=Free
  swing2LimitDegrees=45
  swing2Motion=Free
  twistLimitDegrees=45
  twistMotion=Free
 }
 {
  bone1Name=tshirt_A_l_dyn_3
  bone2Name=thigh_l
  swing1LimitDegrees=45
  swing1Motion=Free
  swing2LimitDegrees=45
  swing2Motion=Free
  twistLimitDegrees=45
  twistMotion=Free
 }
 {
  bone1Name=tshirt_A_r_dyn_3
  bone2Name=tshirt_A_l_dyn_3
  swing1LimitDegrees=45
  swing1Motion=Free
  swing2LimitDegrees=45
  swing2Motion=Free
  twistLimitDegrees=45
  twistMotion=Free
 }
 {
  bone1Name=tshirt_A_l_dyn_31
  bone2Name=tshirt_A_l_dyn_32
  swing1LimitDegrees=45
  swing1Motion=Free
  swing2LimitDegrees=45
  swing2Motion=Free
  twistLimitDegrees=45
  twistMotion=Free
 }
 {
  bone1Name=tshirt_A_l_dyn_31
  bone2Name=tshirt_mid_l_dyn_3
  swing1LimitDegrees=45
  swing1Motion=Free
  swing2LimitDegrees=45
  swing2Motion=Free
  twistLimitDegrees=45
  twistMotion=Free
 }
 {
  bone1Name=tshirt_A_r_dyn_3
  bone2Name=tshirt_mid_l_dyn_3
  swing1LimitDegrees=45
  swing1Motion=Free
  swing2LimitDegrees=45
  swing2Motion=Free
  twistLimitDegrees=45
  twistMotion=Free
 }
]
