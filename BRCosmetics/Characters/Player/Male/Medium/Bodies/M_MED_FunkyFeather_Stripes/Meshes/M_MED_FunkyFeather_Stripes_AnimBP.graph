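# Animation Blueprint for the FunkyFeather Stripes medium male body cosmetic, adding secondary motion to clothing and accessories.
# It copies the attached parent's pose through a Control Rig, then simulates necklace and T-shirt (RigidBody) and sleeves and leg chains (AnimDynamics).
# A second, frontend-aware chain Control Rig finishes the pose.
# Each update sets shared gravity from a parameter asset and lowers sleeve angular velocity and acceleration caps in the frontend.
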AnimBlueprint /BRCosmetics/Characters/Player/Male/Medium/Bodies/M_MED_FunkyFeather_Stripes/Meshes/M_MED_FunkyFeather_Stripes_AnimBP
target_skeleton=&/M_MED_FunkyFeather_Stripes_Skeleton
graphs [
 AnimGraph {nodes [
   n1 = Get Sim Space Settings Sleeves() -> (Sim Space Settings Sleeves:Anim Phys Sim Space Settings Structure)
   n2 = Control Rig(Source=Copy Pose From Mesh() {bUseAttachedParent=true}) {controlRigAssetReference {blueprintRigClass=&/M_MED_FunkyFeather_Stripes_CtrlRig.M_MED_FunkyFeather_Stripes_CtrlRig_C}}
   n3 = RigidBody!(ComponentPose=Local To Component(LocalPose=@n2), Alpha, OverrideWorldGravity:Vector=Get Gravity Override(), ExternalForce) -> (Pose) {
    bOverrideWorldGravity=true
    overridePhysicsAsset=&/M_MED_FunkyFeather_Stripes_Necklase_Physics
    simulationSpace=BaseBoneSpace
    baseBoneRef {boneName=spine_04}
    componentAppliedLinearAccClamp=<0 0 0>
    simSpaceSettings {
     dampingAlpha=0
     maxAngularAcceleration=10
     maxAngularVelocity=6
     maxLinearAcceleration=1200
     maxLinearVelocity=500
     worldAlpha=1
     externalLinearDragV=<0.6 0.6 0.6>
    }
   }
   n4 = RigidBody(ComponentPose=@n3, OverrideWorldGravity=Get Gravity Override()) {
    bOverrideWorldGravity=true
    overridePhysicsAsset=&/M_MED_FunkyFeather_Stripes_TShirt_Physics
    simulationSpace=BaseBoneSpace
    baseBoneRef {boneName=pelvis}
    componentAppliedLinearAccClamp=<0 0 0>
    simSpaceSettings {
     dampingAlpha=0
     maxAngularAcceleration=9
     maxAngularVelocity=6
     maxLinearAcceleration=1200
     maxLinearVelocity=500
     worldAlpha=1
     externalLinearDragV=<0.6 0.6 0.6>
    }
   }
   n5 = Get Sim Space Settings Sleeves()
   n6 = Anim Dynamics - Bone: sleeve_l_C_dyn_1!(
    ComponentPose=@n4
    Alpha=1
    GravityOverride:Vector=Get Gravity Override()
    SimSpaceSettings:Anim Phys Sim Space Settings Structure=@n5
    ExternalForce="0,0,0"
   ) -> (Pose) {
    angularDampingOverride=1
    angularSpringConstant=5000
    bAngularSpring=true
    bOverrideAngularDamping=true
    bUseGravityOverride=true
    simulationSpace=RootRelative
    boundBone {boneName=sleeve_l_C_dyn_1}
    componentAppliedLinearAccClamp=<3000 3000 3000>
    componentLinearAccScale=<0.5 0.5 0.5>
    componentLinearVelScale=<0.5 0.5 0.5>
    physicsBodyDefinitions [{
      boundBone {boneName=sleeve_l_C_dyn_1}
      boxExtents=<11 11 11>
      constraintSetup {angularLimitsMax=<0 6 5>, angularLimitsMin=<0 -6 -5>, angularTarget=<1 0 0>}
      localJointOffset=<4 0 0>
     }]
    relativeSpaceBone {boneName=upperarm_twist_01_r}
    simSpaceSettings {maxAngularAcceleration=10, maxAngularVelocity=7, simSpaceAngularAlpha=1}
   }
   n7 = Anim Dynamics - Bone: sleeve_l_A_dyn_1!(
    ComponentPose=@n6
    Alpha=1
    GravityOverride:Vector=Get Gravity Override()
    SimSpaceSettings:Anim Phys Sim Space Settings Structure=@n5
    ExternalForce="0,0,0"
   ) -> (Pose) {
    angularDampingOverride=1
    angularSpringConstant=500
    bAngularSpring=true
    bOverrideAngularDamping=true
    bUseGravityOverride=true
    simulationSpace=RootRelative
    boundBone {boneName=sleeve_l_A_dyn_1}
    componentAppliedLinearAccClamp=<3000 3000 3000>
    componentLinearAccScale=<0.5 0.5 0.5>
    componentLinearVelScale=<0.5 0.5 0.5>
    physicsBodyDefinitions [{
      boundBone {boneName=sleeve_l_A_dyn_1}
      boxExtents=<15 15 15>
      constraintSetup {angularLimitsMax=<0 5 5>, angularLimitsMin=<0 -5 -5>, angularTarget=<1 0 0>}
      localJointOffset=<4 0 0>
     }]
    relativeSpaceBone {boneName=upperarm_twist_01_r}
    simSpaceSettings {maxAngularAcceleration=10, maxAngularVelocity=7, simSpaceAngularAlpha=1}
   }
   n8 = Anim Dynamics - Bone: sleeve_l_B_dyn_1!(
    ComponentPose=@n7
    Alpha=1
    GravityOverride:Vector=Get Gravity Override()
    SimSpaceSettings:Anim Phys Sim Space Settings Structure=@n5
    ExternalForce="0,0,0"
   ) -> (Pose) {
    angularDampingOverride=1
    angularSpringConstant=5000
    bAngularSpring=true
    bOverrideAngularDamping=true
    bUseGravityOverride=true
    simulationSpace=RootRelative
    boundBone {boneName=sleeve_l_B_dyn_1}
    componentAppliedLinearAccClamp=<3000 3000 3000>
    componentLinearAccScale=<0.5 0.5 0.5>
    componentLinearVelScale=<0.5 0.5 0.5>
    physicsBodyDefinitions [{
      boundBone {boneName=sleeve_l_B_dyn_1}
      boxExtents=<11 11 11>
      constraintSetup {angularLimitsMax=<0 5 5>, angularLimitsMin=<0 -5 -5>, angularTarget=<1 0 0>}
      localJointOffset=<4 0 0>
     }]
    relativeSpaceBone {boneName=upperarm_twist_01_r}
    simSpaceSettings {maxAngularAcceleration=10, maxAngularVelocity=7, simSpaceAngularAlpha=1}
   }
   n9 = Anim Dynamics - Bone: sleeve_l_D_dyn_1!(
    ComponentPose=@n8
    Alpha=1
    GravityOverride:Vector=Get Gravity Override()
    SimSpaceSettings:Anim Phys Sim Space Settings Structure=@n5
    ExternalForce="0,0,0"
   ) -> (Pose) {
    angularDampingOverride=1
    angularSpringConstant=500
    bAngularSpring=true
    bOverrideAngularDamping=true
    bUseGravityOverride=true
    simulationSpace=RootRelative
    boundBone {boneName=sleeve_l_D_dyn_1}
    componentAppliedLinearAccClamp=<3000 3000 3000>
    componentLinearAccScale=<0.5 0.5 0.5>
    componentLinearVelScale=<0.5 0.5 0.5>
    physicsBodyDefinitions [{
      boundBone {boneName=sleeve_l_D_dyn_1}
      boxExtents=<15 15 15>
      constraintSetup {angularLimitsMax=<0 5 5>, angularLimitsMin=<0 -5 -5>, angularTarget=<1 0 0>}
      localJointOffset=<4 0 0>
     }]
    relativeSpaceBone {boneName=upperarm_twist_01_r}
    simSpaceSettings {maxAngularAcceleration=10, maxAngularVelocity=7, simSpaceAngularAlpha=1}
   }
   n10 = Anim Dynamics - Bone: sleeve_l_E_dyn_1!(
    ComponentPose=@n9
    Alpha=1
    GravityOverride:Vector=Get Gravity Override()
    SimSpaceSettings:Anim Phys Sim Space Settings Structure=@n5
    ExternalForce="0,0,0"
   ) -> (Pose) {
    angularDampingOverride=1
    angularSpringConstant=500
    bAngularSpring=true
    bOverrideAngularDamping=true
    bUseGravityOverride=true
    simulationSpace=RootRelative
    boundBone {boneName=sleeve_l_E_dyn_1}
    componentAppliedLinearAccClamp=<3000 3000 3000>
    componentLinearAccScale=<0.5 0.5 0.5>
    componentLinearVelScale=<0.5 0.5 0.5>
    physicsBodyDefinitions [{
      boundBone {boneName=sleeve_l_E_dyn_1}
      boxExtents=<15 15 15>
      constraintSetup {angularLimitsMax=<0 5 6>, angularLimitsMin=<0 -5 -5>, angularTarget=<1 0 0>}
      localJointOffset=<4 0 0>
     }]
    relativeSpaceBone {boneName=upperarm_twist_01_r}
    simSpaceSettings {maxAngularAcceleration=10, maxAngularVelocity=7, simSpaceAngularAlpha=1}
   }
   n11 = Anim Dynamics - Bone: sleeve_r_C_dyn_1!(
    ComponentPose=@n10
    Alpha=1
    GravityOverride:Vector=Get Gravity Override()
    SimSpaceSettings:Anim Phys Sim Space Settings Structure=@n1
    ExternalForce="0,0,0"
   ) -> (Pose) {
    angularDampingOverride=1
    angularSpringConstant=5000
    bAngularSpring=true
    bOverrideAngularDamping=true
    bUseGravityOverride=true
    simulationSpace=RootRelative
    boundBone {boneName=sleeve_r_C_dyn_1}
    componentAppliedLinearAccClamp=<3000 3000 3000>
    componentLinearAccScale=<0.5 0.5 0.5>
    componentLinearVelScale=<0.5 0.5 0.5>
    physicsBodyDefinitions [{
      boundBone {boneName=sleeve_r_C_dyn_1}
      boxExtents=<11 11 11>
      constraintSetup {angularLimitsMax=<0 5 5>, angularLimitsMin=<0 -5 -5>, angularTarget=<1 0 0>}
      localJointOffset=<4 0 0>
     }]
    relativeSpaceBone {boneName=upperarm_twist_01_r}
    simSpaceSettings {maxAngularAcceleration=10, maxAngularVelocity=7, simSpaceAngularAlpha=1}
   }
   n12 = Anim Dynamics - Bone: sleeve_r_A_dyn_1!(
    ComponentPose=@n11
    Alpha=1
    GravityOverride:Vector=Get Gravity Override()
    SimSpaceSettings:Anim Phys Sim Space Settings Structure=@n1
    ExternalForce="0,0,0"
   ) -> (Pose) {
    angularDampingOverride=1
    angularSpringConstant=500
    bAngularSpring=true
    bOverrideAngularDamping=true
    bUseGravityOverride=true
    simulationSpace=RootRelative
    boundBone {boneName=sleeve_r_A_dyn_1}
    componentAppliedLinearAccClamp=<3000 3000 3000>
    componentLinearAccScale=<0.5 0.5 0.5>
    componentLinearVelScale=<0.5 0.5 0.5>
    physicsBodyDefinitions [{
      boundBone {boneName=sleeve_r_A_dyn_1}
      boxExtents=<15 15 15>
      constraintSetup {angularLimitsMax=<0 5 5>, angularLimitsMin=<0 -5 -5>, angularTarget=<1 0 0>}
      localJointOffset=<4 0 0>
     }]
    relativeSpaceBone {boneName=upperarm_twist_01_r}
    simSpaceSettings {maxAngularAcceleration=10, maxAngularVelocity=7, simSpaceAngularAlpha=1}
   }
   n13 = Anim Dynamics - Bone: sleeve_r_B_dyn_1!(
    ComponentPose=@n12
    Alpha=1
    GravityOverride:Vector=Get Gravity Override()
    SimSpaceSettings:Anim Phys Sim Space Settings Structure=@n1
    ExternalForce="0,0,0"
   ) -> (Pose) {
    angularDampingOverride=1
    angularSpringConstant=5000
    bAngularSpring=true
    bOverrideAngularDamping=true
    bUseGravityOverride=true
    simulationSpace=RootRelative
    boundBone {boneName=sleeve_r_B_dyn_1}
    componentAppliedLinearAccClamp=<3000 3000 3000>
    componentLinearAccScale=<0.5 0.5 0.5>
    componentLinearVelScale=<0.5 0.5 0.5>
    physicsBodyDefinitions [{
      boundBone {boneName=sleeve_r_B_dyn_1}
      boxExtents=<11 11 11>
      constraintSetup {angularLimitsMax=<0 0 5>, angularLimitsMin=<0 -5 -5>, angularTarget=<1 0 0>}
      localJointOffset=<4 0 0>
     }]
    relativeSpaceBone {boneName=upperarm_twist_01_r}
    simSpaceSettings {maxAngularAcceleration=10, maxAngularVelocity=7, simSpaceAngularAlpha=1}
   }
   n14 = Anim Dynamics - Bone: sleeve_r_D_dyn_1!(
    ComponentPose=@n13
    Alpha=1
    GravityOverride:Vector=Get Gravity Override()
    SimSpaceSettings:Anim Phys Sim Space Settings Structure=@n1
    ExternalForce="0,0,0"
   ) -> (Pose) {
    angularDampingOverride=1
    angularSpringConstant=500
    bAngularSpring=true
    bOverrideAngularDamping=true
    bUseGravityOverride=true
    simulationSpace=RootRelative
    boundBone {boneName=sleeve_r_D_dyn_1}
    componentAppliedLinearAccClamp=<3000 3000 3000>
    componentLinearAccScale=<0.5 0.5 0.5>
    componentLinearVelScale=<0.5 0.5 0.5>
    physicsBodyDefinitions [{
      boundBone {boneName=sleeve_r_D_dyn_1}
      boxExtents=<15 15 15>
      constraintSetup {angularLimitsMax=<0 5 5>, angularLimitsMin=<0 -5 -5>, angularTarget=<1 0 0>}
      localJointOffset=<4 0 0>
     }]
    relativeSpaceBone {boneName=upperarm_twist_01_r}
    simSpaceSettings {maxAngularAcceleration=10, maxAngularVelocity=7, simSpaceAngularAlpha=1}
   }
   n15 = Anim Dynamics - Bone: sleeve_r_E_dyn_1!(
    ComponentPose=@n14
    Alpha=1
    GravityOverride:Vector=Get Gravity Override()
    SimSpaceSettings:Anim Phys Sim Space Settings Structure=@n1
    ExternalForce="0,0,0"
   ) -> (Pose) {
    angularDampingOverride=1
    angularSpringConstant=500
    bAngularSpring=true
    bOverrideAngularDamping=true
    bUseGravityOverride=true
    simulationSpace=RootRelative
    boundBone {boneName=sleeve_r_E_dyn_1}
    componentAppliedLinearAccClamp=<3000 3000 3000>
    componentLinearAccScale=<0.5 0.5 0.5>
    componentLinearVelScale=<0.5 0.5 0.5>
    physicsBodyDefinitions [{
      boundBone {boneName=sleeve_r_E_dyn_1}
      boxExtents=<11 11 11>
      constraintSetup {angularLimitsMax=<0 5 5>, angularLimitsMin=<0 -5 -5>, angularTarget=<1 0 0>}
      localJointOffset=<4 0 0>
     }]
    relativeSpaceBone {boneName=upperarm_twist_01_r}
    simSpaceSettings {maxAngularAcceleration=10, maxAngularVelocity=7, simSpaceAngularAlpha=1}
   }
   n16 = Anim Dynamics - Bone: chain_leg_A_1!(ComponentPose=@n15, Alpha=1, GravityOverride:Vector=Get Gravity Override(), ExternalForce="0,0,0") -> (Pose) {
    angularDampingOverride=1
    angularSpringConstant=2000
    bAngularSpring=true
    bOverrideAngularDamping=true
    bUseGravityOverride=true
    simulationSpace=RootRelative
    boundBone {boneName=chain_leg_A_1}
    componentAppliedLinearAccClamp=<3000 3000 3000>
    componentLinearAccScale=<1 1 1>
    componentLinearVelScale=<1 1 1>
    physicsBodyDefinitions [{
      boundBone {boneName=chain_leg_A_1}
      boxExtents=<15 15 15>
      constraintSetup {angularLimitsMax=<0 0 3>, angularLimitsMin=<0 0 -25>, angularTarget=<1 0 0>}
      localJointOffset=<5 0 0>
     }]
    relativeSpaceBone {boneName=upperarm_twist_01_r}
    simSpaceSettings {maxAngularAcceleration=35, maxAngularVelocity=7, simSpaceAngularAlpha=1}
   }
   n17 = Anim Dynamics - Bone: chain_leg_B_1!(ComponentPose=@n16, Alpha=1, GravityOverride:Vector=Get Gravity Override(), ExternalForce="0,0,0") -> (Pose) {
    angularDampingOverride=1
    angularSpringConstant=4000
    bAngularSpring=true
    bOverrideAngularDamping=true
    bUseGravityOverride=true
    simulationSpace=RootRelative
    boundBone {boneName=chain_leg_B_1}
    componentAppliedLinearAccClamp=<3000 3000 3000>
    componentLinearAccScale=<1 1 1>
    componentLinearVelScale=<1 1 1>
    physicsBodyDefinitions [{
      boundBone {boneName=chain_leg_B_1}
      boxExtents=<15 15 15>
      constraintSetup {angularLimitsMax=<0 0 3>, angularLimitsMin=<0 0 -25>, angularTarget=<1 0 0>}
      localJointOffset=<5 0 0>
     }]
    relativeSpaceBone {boneName=upperarm_twist_01_r}
    simSpaceSettings {maxAngularAcceleration=35, maxAngularVelocity=7, simSpaceAngularAlpha=1}
   }
   n18 = Control Rig!(Source=Component To Local(ComponentPose=@n17), Alpha, IsFrontend:bool=Get bIsInFrontEnd()) -> (Pose) {controlRigAssetReference {blueprintRigClass=&/M_MED_FunkyFeather_Stripes_CtrlRig_Chain.M_MED_FunkyFeather_Stripes_CtrlRig_Chain_C}}
   Output Pose(Result=@n18) {layerGroup=None}
  ]}
 EventGraph {nodes [
   Event BlueprintUpdateAnimation()
   > n1 = "Gravity Override State Function (Parameter Asset) (Nativized)"(
    CharacterPartAnimInstance=Self-Reference()
    GameTimeInSeconds=GetGameTimeInSeconds()
    Parameters=&/F_MED_FynkyFeather_Stripes_Params
   )
   > Set Gravity Override(Gravity Override=@n1)
   > n2 = Branch(Condition=Get bIsInFrontEnd())
   > Set!(
    execute
    Sim Space Settings Sleeves_SimSpaceAngularAlpha:float=1
    Sim Space Settings Sleeves_MaxAngularVelocity:float=2
    Sim Space Settings Sleeves_MaxAngularAcceleration:float=4
    Sim Space Settings Sleeves_ExternalAngularVelocity:Vector="0,0,0"
   ) -> (then, Output_Get:Anim Phys Sim Space Settings Structure)
   Set(
    execute=@n2.else
    Sim Space Settings Sleeves_SimSpaceAngularAlpha=1
    Sim Space Settings Sleeves_MaxAngularVelocity=7
    Sim Space Settings Sleeves_MaxAngularAcceleration=10
    Sim Space Settings Sleeves_ExternalAngularVelocity="0,0,0"
   )
  ]}
]
variables [
 Gravity Override {type=Vector, default=<0 0 -980>}
 Sim Space Settings Sleeves {
  type=AnimPhysSimSpaceSettings
  default {maxAngularAcceleration=7, maxAngularVelocity=10, simSpaceAngularAlpha=1, externalAngularVelocity=<0 0 0>}
 }
]
