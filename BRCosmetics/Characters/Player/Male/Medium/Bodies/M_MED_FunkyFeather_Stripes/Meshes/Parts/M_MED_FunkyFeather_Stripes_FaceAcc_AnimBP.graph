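# Animation Blueprint for the face accessory mesh of the Male Medium FunkyFeather Stripes character body, adding secondary motion to its strap.
# It copies the pose from the attached parent mesh, then runs an Anim Dynamics node on the strap bone before output.
# The strap simulates as a single box body with gravity, angular spring, damping and limited angular range.
# The event graph holds no logic.
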AnimBlueprint /BRCosmetics/Characters/Player/Male/Medium/Bodies/M_MED_FunkyFeather_Stripes/Meshes/Parts/M_MED_FunkyFeather_Stripes_FaceAcc_AnimBP
target_skeleton=&/M_MED_FunkyFeather_Stripes_FaceAcc_Skeleton
graphs [
 AnimGraph {nodes [
   n1 = Anim Dynamics - Bone: strap(ComponentPose=Local To Component(LocalPose=Copy Pose From Mesh() {bUseAttachedParent=true}), Alpha=1, ExternalForce="0,0,0") {
    angularDampingOverride=1
    angularSpringConstant=6000
    bAngularSpring=true
    bOverrideAngularDamping=true
    bUseGravityOverride=true
    linearSpringConstant=5
    alphaScaleBias {bias=-0.0001, scale=1.0001}
    boundBone {boneName=strap}
    componentAppliedLinearAccClamp=<1000 1000 1000>
    componentLinearAccScale=<1 1 1>
    componentLinearVelScale=<1 1 1>
    gravityOverride {z=-980}
    physicsBodyDefinitions [{
      boundBone {boneName=strap}
      boxExtents=<12 12 12>
      constraintSetup {angularLimitsMax=<0 5 20>, angularLimitsMin=<0 -15 -20>, angularTarget=<1 0 0>}
      localJointOffset=<2 0 0>
     }]
    relativeSpaceBone {boneName=head}
    simSpaceSettings {maxAngularAcceleration=10, maxAngularVelocity=1, simSpaceAngularAlpha=1}
   }
   Output Pose(Result=Component To Local(ComponentPose=@n1)) {layerGroup=None}
  ]}
 EventGraph {nodes [Event BlueprintUpdateAnimation()]}
]
variables []
